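# Translucent, unlit, two-sided Niagara sprite material for the Grapple Gloves charge-up gameplay cue effect.
# A scrolling noise texture slightly distorts the UVs of the MaskTexture parameter, whose red channel passes through a same-input lerp.
# That mask blends dark to bright blue, tinted by particle color into MF_FX_EmissiveAdjustments, and also multiplies particle alpha.
# A panning inky smoke sample feeds no other node.
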Material /GrappleGloves/GameplayCues/FX/M_GrappleGloves_ChargeUp
graph {
 properties {
  bUsedWithNiagaraSprites=true
  blendMode=BLEND_Translucent
  pixelDepthOffsetMode=PDOM_Legacy
  shadingModel=MSM_Unlit
  twoSided=true
  usedShadingModels=Unlit
 }
 nodes [
  n1 = Constant3Vector() {constant=<rgba 0 0.0059 0.0521 0>}
  n2 = Constant3Vector() {constant=<rgba 0 0.22 0.9913 0>}
  n3 = Panner() {speedY=-0.15}
  n4 = Multiply(A=ComponentMask(None=TextureSample(UVs=@n3) {texture=&T_NoisePack_01}.RGB) {g=true, r=true}) {constB=0.04}
  n5 = TextureCoordinate()
  n6 = TextureSampleParameter2D(UVs=Panner(Coordinate=Add(A=@n4, B=@n5))) {parameterName=MaskTexture, texture=/Game/Effects/Fort_Effects/BluGlow/ZoomNoise}
  n7 = LinearInterpolate(A=@n6.R, B=@n6.R)
  n8 = ParticleColor()
  MaterialFunctionCall(Emissive=Multiply(A=LinearInterpolate(A=@n1, B=@n2, Alpha=@n7), B=@n8.RGB)) {materialFunction=&MF_FX_EmissiveAdjustments}
  Multiply(A=@n8.A, B=@n7)
  TextureSample(UVs=Panner() {speedY=1}) {texture=&T_Inky_Smoke_Tile}
 ]
}
